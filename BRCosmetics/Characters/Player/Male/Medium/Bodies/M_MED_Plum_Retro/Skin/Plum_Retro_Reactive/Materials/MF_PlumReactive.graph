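# Material function for the Plum Retro reactive skin that adds looping, hex-patterned energy pulses to a material's emissive.
# A noise-distorted pre-skinned sphere mask, or bounds-relative height under the PLumBackpack switch, provides the gradient.
# Two time-offset Frac sweeps cut bands with hex-jittered width, lit by Fresnel glow and modulating base emissive.
# Extra hit masks and Blueprint-timed pulse nodes do not reach the output.
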MaterialFunction /BRCosmetics/Characters/Player/Male/Medium/Bodies/M_MED_Plum_Retro/Skin/Plum_Retro_Reactive/Materials/MF_PlumReactive
graph {nodes [
  n1 = VectorParameter() {parameterName=LX_TransitionNoiseTiling, defaultValue=<rgba 12 12 12 1>}
  n2 = TextureSample(UVs=Multiply(A=TextureCoordinate(), B=ComponentMask(None=@n1.RGB) {g=true, r=true})) {
   samplerType=SAMPLERTYPE_Masks
   texture=/BRCosmetics/Characters/Player/Male/Medium/Bodies/M_MED_Plum_Retro/Skin/Plum_Retro_Reactive/Textures/T_PlumHexPattern
  }
  n3 = Time()
  n4 = ScalarParameter() {defaultValue=0.15, parameterName=pulseSpeed}
  n5 = Frac(None=Multiply(A=@n3, B=@n4) {constB=0.25})
  n6 = Add(
   A=ScalarParameter() {defaultValue=0.025, parameterName=LX_TransitionEdgeRadius}
   B=Multiply(A=@n2.A, B=ScalarParameter() {defaultValue=0.1, parameterName=LX_EdgeLineOffset_Amt})
  )
  n7 = ScalarParameter() {parameterName=LX_TransitionEdgeHardness}
  n8 = PreSkinnedPosition()
  n9 = FunctionInput(Preview=ShaderStageSwitch(PixelShader=VertexInterpolator(VS=@n8).PS, VertexShader=@n8)) {bUsePreviewValueAsDefault=true, inputName=PreskinnedPosition}
  n10 = MaterialFunctionCall(
   TextureObject=TextureObject() {samplerType=SAMPLERTYPE_Masks, texture=/Game/Global/Textures/Default/Noise/T_Noise_Cloud_Mask}
   TextureSize=VectorParameter() {parameterName=PlumBaseNoiseScale, defaultValue=<rgba 512 512 512 1>}.RGB
   WorldPosition=@n9
  ) {materialFunction=&WorldAlignedTexture}
  n11 = Multiply(A=Subtract(A=Multiply(A=@n10.XYZ Texture) {constB=2})) {constB=64}
  n12 = Add(A=@n11, B=@n9)
  n13 = MaterialFunctionCall() {materialFunction=&ObjectLocalBounds}
  n14 = VectorParameter() {parameterName=Plum_MaskPosition_01, defaultValue=<rgba 0 8 128 1>}
  n15 = StaticSwitchParameter(
   True=ComponentMask(None=Divide(A=@n12, B=@n13.Local Bounds Max)) {b=true}
   False=SphereMask(
    A=@n9
    B=Add(A=@n11, B=@n14.RGB)
    Radius=ScalarParameter() {defaultValue=192, parameterName=Plum_HitRadius}
    Hardness=ScalarParameter() {parameterName=LX_MaskHardness}
   )
  ) {parameterName=PLumBackpack}
  n16 = Frac(None=Multiply(A=Add(A=@n3) {constB=1.5}, B=@n4) {constB=0.25})
  n17 = Max(A=SphereMask(A=@n15, B=OneMinus(None=@n5), Radius=@n6, Hardness=@n7), B=SphereMask(A=@n15, B=OneMinus(None=@n16), Radius=@n6, Hardness=@n7))
  n18 = VectorParameter() {parameterName=LX_TransitionEdge_Color, defaultValue=<rgba 0.4382 0 1 1>}
  n19 = Multiply(A=Multiply(A=@n2.R, B=@n17), B=@n18.RGB)
  n20 = ScalarParameter() {defaultValue=1, parameterName=FresnelExponent_01}
  n21 = PixelNormalWS()
  n22 = LinearInterpolate(
   A=ScalarParameter() {defaultValue=1, parameterName=LX_TransitionEdge_MinGlowStrength}
   B=ScalarParameter() {defaultValue=10, parameterName=LX_TransitionEdge_MaxGlowStrength}
   Alpha=Power(Base=Fresnel(ExponentIn=@n20, Normal=@n21))
  )
  n23 = Fresnel(ExponentIn=ScalarParameter() {defaultValue=1, parameterName=FresnelExponent_02}, Normal=@n21)
  n24 = OneMinus(None=@n2.R)
  n25 = Add(A=@n2.A, B=Multiply(A=Time(), B=ScalarParameter() {defaultValue=0.25, parameterName=HexDotSpeed}) {constB=0.5})
  n26 = Clamp(
   None=ConstantBiasScale(None=Sine(None=@n25))
   Min=ScalarParameter() {parameterName=HexPatternMin}
   Max=ScalarParameter() {defaultValue=0.25, parameterName=HexPatternMax}
  )
  n27 = Subtract(
   A=Power(Base=@n23, Exp=ScalarParameter() {defaultValue=1, parameterName=PlumFresnelPower}) {constExponent=1.5}
   B=Multiply(A=Multiply(A=@n24, B=@n26), B=ScalarParameter() {defaultValue=0.5, parameterName=FresnelNoise_Strength})
  )
  n28 = Multiply(A=Multiply(A=Clamp(None=@n27), B=@n18.RGB), B=ScalarParameter() {defaultValue=4, parameterName=EdgeGlowStrength})
  n29 = Add(A=Power(Base=Multiply(A=@n19, B=@n22)), B=Multiply(A=@n17, B=Power(Base=@n28)))
  Add(A=@n29)
  n30 = GetMaterialAttributes(None=FunctionInput() {inputName=Material, inputType=FunctionInput_MaterialAttributes}) {attributeGetTypes [#1, #2]}
  n31 = Multiply(A=@n30.EmissiveColor, B=Clamp(None=@n17, Min=ScalarParameter() {defaultValue=0.25, parameterName=BaseEmissiveStrength}))
  n32 = ComponentMask(None=@n31) {b=true}
  Ceil(None=@n32)
  n33 = Subtract(A=Time(), B=ScalarParameter() {defaultValue=-9.999999980506448e+18, parameterName=Plum_GameTimeFromBP_02})
  n34 = Multiply(A=ScalarParameter() {defaultValue=0.5, parameterName=Plum_PulseDuration}) {constB=0.5}
  n35 = Saturate(None=Divide(A=Subtract(A=@n33, B=@n34) {constB=0.5}, B=ScalarParameter() {defaultValue=2, parameterName=Plum_fresnelDuration}))
  ConstantBiasScale(None=Sine(None=Subtract(A=@n35) {constB=0.25}))
  n36 = Multiply(A=@n29, B=ScalarParameter() {defaultValue=0.25, parameterName=Plum_GlobalEmissiveMult})
  FunctionOutput(None=SetMaterialAttributes!(MaterialAttributes=@n30.MaterialAttributes, EmissiveColor=Add(A=@n31, B=@n36)) {attributeSetTypes [#1]})
  n37 = VectorParameter() {parameterName=Plum_MaskPosition_02, defaultValue=<rgba 32 0 96 1>}
  n38 = ScalarParameter() {defaultValue=128, parameterName=Plum_HitRadius_02}
  n39 = ScalarParameter() {parameterName=LX_MaskHardness}
  n40 = SphereMask(
   A=@n9
   B=VectorParameter() {parameterName=Plum_MaskPosition_03, defaultValue=<rgba -32 0 140 1>}.RGB
   Radius=ScalarParameter() {defaultValue=128, parameterName=Plum_HitRadius_03}
   Hardness=ScalarParameter() {parameterName=LX_MaskHardness}
  )
  Max(A=Max(A=@n15, B=SphereMask(A=@n9, B=@n37.RGB, Radius=@n38, Hardness=@n39)), B=@n40)
  Multiply(A=Round(None=@n32))
  Multiply(A=Clamp() {minDefault=0.75}, B=@n22)
  n41 = ScalarParameter() {defaultValue=1, parameterName=Pulse_03_Timeoffset}
  Multiply(A=Ceil(None=@n40), B=@n41)
  Multiply(A=@n40, B=@n41) {constB=0.2}
  Multiply(A=@n22, B=ScalarParameter() {defaultValue=0.1, parameterName=BrightnessFalloff})
  Normalize()
  OneMinus(None=@n23)
  ScalarParameter() {defaultValue=1, parameterName=Plum_ShieldValue}
  ScalarParameter() {defaultValue=0.5, parameterName=pulseSpeed_02}
  TextureSample(UVs=Multiply(A=TextureCoordinate(), B=ScalarParameter() {defaultValue=0.25, parameterName=Plum_baseNoiseScale}) {constB=0.25}) {samplerType=SAMPLERTYPE_Masks, texture=&T_Noise_Cloud_Mask}
 ]}
