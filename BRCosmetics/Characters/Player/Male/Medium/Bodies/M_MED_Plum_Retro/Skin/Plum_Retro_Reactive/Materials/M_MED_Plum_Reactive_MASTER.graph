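# Master skeletal-mesh character material for the reactive Plum Retro male skin, built on material attributes.
# The shared PreFX character function feeds a chain of switchable layers: masked thin film, masked clear coat, Plum reactive effect (on by default), then cape wind.
# Emissive fresnel always follows, then the shared PostFX function finishes the chain.
# A skin FX mask supplies selectable channels; the reactive and wind functions use pre-skinned local position.
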Material /BRCosmetics/Characters/Player/Male/Medium/Bodies/M_MED_Plum_Retro/Skin/Plum_Retro_Reactive/Materials/M_MED_Plum_Reactive_MASTER
graph {
 properties {
  bUseMaterialAttributes=true
  bUsedWithMeshDeformer=true
  bUsedWithMorphTargets=true
  bUsedWithSkeletalMesh=true
  pixelDepthOffsetMode=PDOM_Legacy
  usedShadingModels=Default Lit
 }
 nodes [
  n1 = MaterialFunctionCall() {materialFunction=&BaseFN_CharacterFunction_PreFX}
  n2 = TextureSample(Tex=TextureObjectParameter() {group=Skin FX, parameterName=SkinFX_Mask})
  n3 = StaticSwitchParameter(
   True=MaterialFunctionCall!(
    Alpha=StaticComponentMaskParameter(None=@n2.RGBA) {defaultR=true, group=Skin FX, parameterName=ThinFilm_Channel}
    Material Attributes=@n1.Result
   ) {materialFunction=/Game/Weapons/WeaponSkins/MaterialFunction/ThinFilmFunction}.Result
   False=@n1.Result
  ) {group=Skin FX, parameterName=Use Thin Film}
  n4 = StaticSwitchParameter(
   True=MaterialFunctionCall!(Alpha=StaticComponentMaskParameter(None=@n2.RGBA) {defaultR=true, group=Skin FX, parameterName=Clear Coat Channel}, In=@n3) {materialFunction=/Game/Weapons/WeaponSkins/MaterialFunction/Clearcoat}.Result
   False=@n3
  ) {group=Skin FX, parameterName=Use Clear Coat}
  n5 = StaticSwitchParameter(
   True=MaterialFunctionCall!(PreskinnedPosition=@n1.Pre skinned Local Position, Material=@n4) {materialFunction=&/MF_PlumReactive}.Result
   False=@n4
  ) {defaultValue=true, parameterName=UsePlumReactive}
  n6 = MaterialFunctionCall!(Preskinned_Input=@n1.Pre skinned Local Position, Material=@n5) {materialFunction=/Game/Characters/Player/Male/Medium/Bodies/M_MED_Plum_Retro/Skin/Plum_Retro_Reactive/Materials/MF_CapeWind}
  MaterialFunctionCall(
   AO=@n1.AO
   Distance Blend=@n1.AO
   Fresnel=@n1.AO
   Gmap=@n1.AO
   Pre Skinned Local Position=@n1.AO
   Material=MaterialFunctionCall!(In=StaticSwitchParameter(True=@n6.Result, False=@n5) {parameterName=UseCapeWind}) {materialFunction=/Game/Weapons/WeaponSkins/MaterialFunction/MF_EmissiveFresnel}.Result
  ) {materialFunction=&BaseFN_CharacterFunction_PostFX}
 ]
}
